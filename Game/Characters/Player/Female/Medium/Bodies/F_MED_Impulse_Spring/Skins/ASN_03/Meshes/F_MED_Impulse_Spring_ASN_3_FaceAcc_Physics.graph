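# Physics asset for the face accessory mesh of the female medium Impulse Spring ASN 3 player skin, defining jointed bodies for its dangling dyn_ bones.
# Two capsules form the head, to which the back and upper bodies each attach directly with limited swing and twist.
# Front A, front B and sideMain chains hang from small base bodies with free rotation to the head, using tight twist limits along each chain.
# Three extra side bodies branch off the sideMain chain.
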PhysicsAsset /Game/Characters/Player/Female/Medium/Bodies/F_MED_Impulse_Spring/Skins/ASN_03/Meshes/F_MED_Impulse_Spring_ASN_3_FaceAcc_Physics
bodies {
 dyn_bk_a [{shapeType=Capsule, length=1, radius=1, center=<1.5 0 0>, rotation=<rot 90 0 0>}]
 dyn_bk_b [{shapeType=Capsule, length=1, radius=1, center=<1.5 0 0>, rotation=<rot 90 0 0>}]
 dyn_fr_A_a [{shapeType=Capsule, length=1, radius=1, center=<1.3645 0 0>, rotation=<rot 90 0 0>}]
 dyn_fr_A_a_base [{shapeType=Capsule, length=0.505, radius=0.505, rotation=<rot 0 -90 0>}]
 dyn_fr_A_b [{shapeType=Capsule, length=2.75, radius=0.75, center=<2.001 0 0>, rotation=<rot 90 0 0>}]
 dyn_fr_B_a [{shapeType=Capsule, length=2, radius=1, center=<2.0149 0 0>, rotation=<rot 90 0 0>}]
 dyn_fr_B_a_base [{shapeType=Capsule, length=0.505, radius=0.505, rotation=<rot 0 -90 0>}]
 dyn_fr_B_b [{shapeType=Capsule, length=3.5, radius=1, center=<2.5921 0 -0.3961>, rotation=<rot 82.1073 0 0>}]
 dyn_sideMain_a [{shapeType=Capsule, length=0, radius=2.5, center=<2.3787 0 0>, rotation=<rot 90 0 0>}]
 dyn_sideMain_a_base [{shapeType=Capsule, length=0.505, radius=0.505, rotation=<rot 0 -90 0>}]
 dyn_sideMain_b [{shapeType=Capsule, length=0.5, radius=2, center=<2.2948 -0.4225 -0.539>, rotation=<rot 75.9292 3.6235 15.3091>}]
 dyn_sideMain_c [{shapeType=Capsule, length=0.25, radius=1.25, center=<1.3951 0 0>, rotation=<rot 90 0 0>}]
 dyn_sideMain_d [{shapeType=Capsule, length=2, radius=0.75, center=<1.8224 0 0>, rotation=<rot 90 0 0>}]
 dyn_side_a [{shapeType=Capsule, length=3.75, radius=0.5, center=<2.3239 0 0>, rotation=<rot 90 0 0>}]
 dyn_side_b [{shapeType=Capsule, length=1.75, radius=1, center=<1.8205 -0.0001 0>, rotation=<rot 90 0 0>}]
 dyn_side_c [{shapeType=Capsule, length=1.25, radius=1, center=<1.5216 0 0>, rotation=<rot 90 0 0>}]
 dyn_up_a [{shapeType=Capsule, length=1, radius=1, center=<1.5 0 0>, rotation=<rot 90 0 0>}]
 dyn_up_b [{shapeType=Capsule, length=1, radius=1, center=<1.5 0 0>, rotation=<rot 90 0 0>}]
 dyn_up_c [{shapeType=Capsule, length=1, radius=1, center=<1.5 0 0>, rotation=<rot 90 0 0>}]
 dyn_up_d [{shapeType=Capsule, length=1, radius=1, center=<1.5 0 0>, rotation=<rot 90 0 0>}]
 dyn_up_e [{shapeType=Capsule, length=1, radius=1, center=<1.5 0 0>, rotation=<rot 90 0 0>}]
 dyn_up_f [{shapeType=Capsule, length=1, radius=1, center=<1.5 0 0>, rotation=<rot 90 0 0>}]
 dyn_up_g [{shapeType=Capsule, length=1, radius=1, center=<1.5 0 0>, rotation=<rot 90 0 0>}]
 head [
  {shapeType=Capsule, length=4.75, radius=6.75, center=<3.5899 1.8346 0>, rotation=<rot 90 0 0>}
  {shapeType=Capsule, length=3, radius=3.75, center=<-1.9732 4.641 0>, rotation=<rot 90 0 7.5623>}
 ]
}
constraints [
 {bone1Name=dyn_bk_b, bone2Name=head, swing1LimitDegrees=25, swing2LimitDegrees=25, twistLimitDegrees=25}
 {bone1Name=dyn_bk_a, bone2Name=head, swing1LimitDegrees=25, swing2LimitDegrees=25, twistLimitDegrees=25}
 {bone1Name=dyn_up_g, bone2Name=head, swing1LimitDegrees=25, swing2LimitDegrees=25, twistLimitDegrees=25}
 {bone1Name=dyn_up_f, bone2Name=head, swing1LimitDegrees=25, swing2LimitDegrees=25, twistLimitDegrees=25}
 {bone1Name=dyn_up_e, bone2Name=head, swing1LimitDegrees=25, swing2LimitDegrees=25, twistLimitDegrees=25}
 {bone1Name=dyn_up_d, bone2Name=head, swing1LimitDegrees=25, swing2LimitDegrees=25, twistLimitDegrees=25}
 {bone1Name=dyn_up_c, bone2Name=head, swing1LimitDegrees=25, swing2LimitDegrees=25, twistLimitDegrees=25}
 {bone1Name=dyn_up_b, bone2Name=head, swing1LimitDegrees=25, swing2LimitDegrees=25, twistLimitDegrees=25}
 {bone1Name=dyn_up_a, bone2Name=head, swing1LimitDegrees=25, swing2LimitDegrees=25, twistLimitDegrees=25}
 {
  bone1Name=dyn_sideMain_a_base
  bone2Name=head
  swing1LimitDegrees=45
  swing1Motion=Free
  swing2LimitDegrees=45
  swing2Motion=Free
  twistLimitDegrees=45
  twistMotion=Free
 }
 {bone1Name=dyn_sideMain_a, bone2Name=dyn_sideMain_a_base, swing1LimitDegrees=25, swing2LimitDegrees=25, twistLimitDegrees=5}
 {bone1Name=dyn_sideMain_b, bone2Name=dyn_sideMain_a, swing1LimitDegrees=25, swing2LimitDegrees=25, twistLimitDegrees=5}
 {bone1Name=dyn_sideMain_c, bone2Name=dyn_sideMain_b, swing1LimitDegrees=25, swing2LimitDegrees=25, twistLimitDegrees=5}
 {bone1Name=dyn_sideMain_d, bone2Name=dyn_sideMain_c, swing1LimitDegrees=25, swing2LimitDegrees=25, twistLimitDegrees=5}
 {bone1Name=dyn_side_b, bone2Name=dyn_sideMain_b, swing1LimitDegrees=45, swing2LimitDegrees=45, twistLimitDegrees=5}
 {bone1Name=dyn_side_a, bone2Name=dyn_sideMain_b, swing1LimitDegrees=45, swing2LimitDegrees=45, twistLimitDegrees=5}
 {bone1Name=dyn_side_c, bone2Name=dyn_sideMain_a, swing1LimitDegrees=45, swing2LimitDegrees=45, twistLimitDegrees=5}
 {
  bone1Name=dyn_fr_A_a_base
  bone2Name=head
  swing1LimitDegrees=45
  swing1Motion=Free
  swing2LimitDegrees=45
  swing2Motion=Free
  twistLimitDegrees=45
  twistMotion=Free
 }
 {
  bone1Name=dyn_fr_B_a_base
  bone2Name=head
  swing1LimitDegrees=45
  swing1Motion=Free
  swing2LimitDegrees=45
  swing2Motion=Free
  twistLimitDegrees=45
  twistMotion=Free
 }
 {bone1Name=dyn_fr_A_a, bone2Name=dyn_fr_A_a_base, swing1LimitDegrees=45, swing2LimitDegrees=45, twistLimitDegrees=5}
 {bone1Name=dyn_fr_A_b, bone2Name=dyn_fr_A_a, swing1LimitDegrees=45, swing2LimitDegrees=45, twistLimitDegrees=5}
 {bone1Name=dyn_fr_B_a, bone2Name=dyn_fr_B_a_base, swing1LimitDegrees=45, swing2LimitDegrees=45, twistLimitDegrees=5}
 {bone1Name=dyn_fr_B_b, bone2Name=dyn_fr_B_a, swing1LimitDegrees=45, swing2LimitDegrees=45, twistLimitDegrees=5}
]
